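# Translucent UI material for the Habanero lobby banner, with an MSDF banner shape and a configurable rank icon silhouette.
# Layers animatable color gradients, outlines with glare and three panning triangle layers, with switches for outline and alternative styling.
# An optional Likes effect adds rising or falling SDF icons from randomized particle grids, plus procedural radial rays.
# AnimColorTransition blends base colors toward Anim_ counterparts and also modulates triangle scale.
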Material /Game/UI/Frontend/Lobby/Habanero/Habanero_Materials/M_Habanero_Banner
graph {
 properties {bAutomaticallySetUsageInEditor=false, blendMode=BLEND_Translucent, materialDomain=MD_UI, usedShadingModels=Default Lit}
 nodes [
  n1 = NamedRerouteUsage()
  n2 = NamedRerouteUsage()
  LinearInterpolate(
   A=@n1
   B=MaterialFunctionCall(Base=@n1, Blend=@n2) {materialFunction=&Blend_Overlay}.Result
   Alpha=Multiply(A=Multiply(A=@n2) {constB=0.2}, B=NamedRerouteUsage())
  )
  n3 = NamedRerouteUsage()
  n4 = Saturate(None=Add(A=NamedRerouteUsage(), B=NamedRerouteUsage()))
  n5 = NamedRerouteUsage()
  n6 = Multiply(A=@n5, B=ComponentMask(None=TextureCoordinate()) {g=true})
  n7 = ScalarParameter() {parameterName=Outline Fade Switch}
  Max(A=Max(A=Multiply(A=@n3, B=@n4), B=LinearInterpolate(A=@n5, B=@n6, Alpha=@n7)), B=Multiply(A=NamedRerouteUsage()) {constB=2})
  n8 = NamedRerouteUsage()
  n9 = NamedRerouteUsage()
  n10 = NamedRerouteUsage()
  n11 = MaterialFunctionCall(GradientStart=Constant() {r=0.164}, GradientEnd=Constant() {r=0.8}, RotationAngle=Constant() {r=0.25}) {materialFunction=&MF_UI_LinearGradient}
  NamedRerouteDeclaration(None=LinearInterpolate(
    A=MaterialFunctionCall(A=@n8, B=@n9, C=@n10, Alpha=@n11.Result) {materialFunction=&3ColorBlend}.Result
    B=NamedRerouteUsage()
    Alpha=NamedRerouteUsage()
   )) {name=Gradient3_Color, nodeColor=<rgba 0.4062 1 0 1>}
  n12 = MaterialFunctionCall(Float2=MaterialFunctionCall() {materialFunction=&GetUserInterfaceUV}.Pixel Size) {materialFunction=&BreakOutFloat2Components}
  NamedRerouteDeclaration(None=Divide(A=@n12.R, B=@n12.R) {constA=537, constB=1015}) {name=AspectRatio, nodeColor=<rgba 0.2188 1 0 1>}
  NamedRerouteDeclaration(None=LinearInterpolate(
    A=VectorParameter() {parameterName=GradientColor_Outer, defaultValue=<rgba 1 0.0685 0 1>}.RGB
    B=VectorParameter() {parameterName=Anim_GradientColor_Outer, defaultValue=<rgba 1 0.0685 0 1>}.RGB
    Alpha=ScalarParameter() {parameterName=AnimColorTransition}
   )) {name=GradientColor_Outer, nodeColor=<rgba 0 1 0.5781 1>}
  n13 = MaterialFunctionCall(Float2=TextureCoordinate()) {materialFunction=&BreakOutFloat2Components}
  n14 = Divide(A=@n13.G, B=NamedRerouteUsage())
  n15 = ScalarParameter() {group=Gradient, parameterName=CircleGradientOffset_X}
  n16 = ScalarParameter() {group=Gradient, parameterName=CircleGradientOffset_Y}
  n17 = MaterialFunctionCall(
   UVs=Subtract(
    A=MaterialFunctionCall(X=@n13.R, Y=@n14) {materialFunction=&MakeFloat2}.Result
    B=MaterialFunctionCall(X=@n15, Y=@n16) {materialFunction=&MakeFloat2}.Result
   )
   Texture Scale=MaterialFunctionCall(
    X=ScalarParameter() {defaultValue=0.7, group=Gradient, parameterName=CircleGradientScale_X}
    Y=ScalarParameter() {defaultValue=0.7, group=Gradient, parameterName=CircleGradientScale_Y}
   ) {materialFunction=&MakeFloat2}.Result
  ) {materialFunction=&ScaleUVsByCenter}
  NamedRerouteDeclaration(None=MaterialFunctionCall(UVs=@n17.UVs, Density=ScalarParameter() {defaultValue=0.5, group=Gradient, parameterName=CircleGradient_Dencity}) {materialFunction=&RadialGradientExponential}.RadialGradientExponential) {name=CircleGradient, nodeColor=<rgba 0 1 0.1797 1>}
  n18 = ScalarParameter() {group=Gradient, parameterName=EllipsisGradientOffset_X}
  n19 = ScalarParameter() {group=Gradient, parameterName=EllipsisGradientOffset_Y}
  n20 = MaterialFunctionCall(
   UVs=Subtract(
    A=TextureCoordinate()
    B=MaterialFunctionCall(X=@n18, Y=@n19) {materialFunction=&MakeFloat2}.Result
   )
   Texture Scale=MaterialFunctionCall(
    X=ScalarParameter() {defaultValue=0.7, group=Gradient, parameterName=EllipsisGradientScale_X}
    Y=ScalarParameter() {defaultValue=1, group=Gradient, parameterName=EllipsisGradientScale_Y}
   ) {materialFunction=&MakeFloat2}.Result
  ) {materialFunction=&ScaleUVsByCenter}
  n21 = ScalarParameter() {defaultValue=1, group=Gradient, parameterName=EllipsisGradient_Dencity}
  NamedRerouteDeclaration(None=Multiply(A=MaterialFunctionCall(UVs=@n20.UVs, Density=@n21) {materialFunction=&RadialGradientExponential}.RadialGradientExponential) {constB=2}) {name=EllipsisGradient, nodeColor=<rgba 1 0 0.1406 1>}
  n22 = MaterialFunctionCall(Float2=TextureCoordinate()) {materialFunction=&BreakOutFloat2Components}
  n23 = MaterialFunctionCall(X=@n22.R, Y=Divide(A=@n22.G, B=NamedRerouteUsage())) {materialFunction=&MakeFloat2}
  n24 = Constant() {r=1.3}
  n25 = ScalarParameter() {parameterName=RankSilhouetteXOffset}
  n26 = ScalarParameter() {defaultValue=-0.352, parameterName=RankSilhouetteYOffset}
  n27 = TextureSampleParameter2D(UVs=Add(
    A=MaterialFunctionCall(UVs=@n23.Result, Texture Scale=@n24) {materialFunction=&ScaleUVsByCenter}.UVs
    B=AppendVector(A=@n25, B=@n26)
   )) {
   parameterName=RankSilhouette
   samplerType=SAMPLERTYPE_LinearColor
   texture=/Game/UI/Frontend/Lobby/Habanero/Habanero_Materials/MaterialInstances/Icons/White/MSDF/T_UI_Icon_Unranked_MSDF
  }
  n28 = MaterialFunctionCall!(MSDF RGB=@n27.RGB) {materialFunction=/FortUILibrary/MaterialFunctions/Shapes/Modifiers/MF_UI_MSDFtoSDF}
  n29 = Constant() {r=0.016}
  n30 = Constant() {r=0.5}
  NamedRerouteDeclaration(None=Saturate(None=MaterialFunctionCall(DistanceField=@n28.SDF, Threshold=@n30, Smoothing=@n29) {materialFunction=&MF_UI_AntiAliasing}.Outer)) {name=RankSillhouette, nodeColor=<rgba 0 0.625 1 1>}
  n31 = ScalarParameter() {defaultValue=0.019, group=Triangle, parameterName=tri-L3-Offset}
  n32 = ScalarParameter() {defaultValue=1.8, group=Triangle, parameterName=ss min}
  n33 = ScalarParameter() {defaultValue=0.35, group=Triangle, parameterName=ss max}
  n34 = Constant() {r=-0.35}
  n35 = Constant()
  n36 = NamedRerouteUsage()
  n37 = Constant() {r=1}
  n38 = MaterialFunctionCall(A=@n37, B=Constant() {r=0.5}, C=@n37, Alpha=ScalarParameter() {parameterName=AnimColorTransition}) {materialFunction=&Lerp_3Color}
  n39 = Multiply(A=ScalarParameter() {defaultValue=0.9048, group=Triangle, parameterName=tri-L3-Scale}, B=@n38.Result)
  n40 = VectorParameter() {parameterName=PannerSpeed, defaultValue=<rgba -0.5 0 -0.45 0>}
  n41 = ScalarParameter() {defaultValue=0.5743, group=Triangle, parameterName=tri-L2-Scale}
  n42 = LinearInterpolate(
   A=Multiply(A=MaterialFunctionCall!(tri-L-Offset=@n31, min=@n32, max=@n33, PannerSpeed_X=@n34, PannerSpeed_Y=@n35, Texture=@n36, EdgeSharp, tri-L-Scale=@n39) {materialFunction=/Game/UI/Frontend/Lobby/Habanero/Habanero_Materials/MaterialFunctions/MF_Habanero_TrianglesEffect}.Result) {constB=0.125}
   Alpha=MaterialFunctionCall(
    tri-L-Offset=ScalarParameter() {defaultValue=-0.6765, group=Triangle, parameterName=tri-L2-Offset}
    min=ScalarParameter() {defaultValue=1.8, group=Triangle, parameterName=ss min}
    max=ScalarParameter() {defaultValue=0.35, group=Triangle, parameterName=ss max}
    PannerSpeed_X=@n40.B
    PannerSpeed_Y=@n40.B
    Texture=NamedRerouteUsage()
    tri-L-Scale=Multiply(A=@n41, B=@n38.Result)
   ) {materialFunction=&MF_Habanero_TrianglesEffect}.Result
  ) {constB=0.25}
  n43 = VectorParameter() {parameterName=PannerSpeed, defaultValue=<rgba -0.5 0 -0.45 0>}
  n44 = MaterialFunctionCall(
   tri-L-Offset=ScalarParameter() {defaultValue=0.2476, group=Triangle, parameterName=tri-L1-Offset}
   min=ScalarParameter() {defaultValue=1.8, group=Triangle, parameterName=ss min}
   max=ScalarParameter() {defaultValue=0.35, group=Triangle, parameterName=ss max}
   PannerSpeed_X=@n43.R
   PannerSpeed_Y=@n43.R
   Texture=NamedRerouteUsage()
   tri-L-Scale=Multiply(A=ScalarParameter() {defaultValue=0.33, group=Triangle, parameterName=tri-L1-Scale}, B=@n38.Result)
  ) {materialFunction=&MF_Habanero_TrianglesEffect}
  NamedRerouteDeclaration(None=Multiply(A=LinearInterpolate(A=@n42, Alpha=@n44.Result)) {constB=2}) {name=Triangles, nodeColor=<rgba 0.9453 0 1 1>}
  NamedRerouteDeclaration(None=TextureObject() {samplerType=SAMPLERTYPE_LinearColor, texture=/Game/UI/Foundation/Items/Textures/T-Card-Hero-Anim-Triangle}) {name=Texture, nodeColor=<rgba 1 0 0.8906 1>}
  n45 = Multiply(A=Constant() {r=5.7646}, B=ScalarParameter() {defaultValue=1, parameterName=Set1_SeedAdjustment})
  n46 = MaterialFunctionCall() {materialFunction=&MF_UI_UVScaling}
  n47 = ScalarParameter() {defaultValue=0.2, group=Likes, parameterName=Set1_Y_PanSpeed}
  n48 = Multiply(A=@n47) {constB=-1}
  n49 = ScalarParameter() {defaultValue=1, group=Likes, parameterName=Up}
  n50 = Constant()
  n51 = MaterialFunctionCall!(
   UV=NamedRerouteUsage()
   Global Offset
   Global Rotation
   Tiling=ScalarParameter() {defaultValue=4, group=Likes, parameterName=Set1_Tile}
   Cell Offset=@n50
   Cell Scale
   Cell Rotation=@n50
   Random Seed=Constant() {r=2.9415}
  ) {materialFunction=/FortUILibrary/MaterialFunctions/Basics/MF_UI_ParticleGrid}
  n52 = AppendVector(
   A=Constant()
   B=Multiply(A=LinearInterpolate(A=@n48, B=@n47, Alpha=@n49), B=SmoothStep(Value=@n51.Random Columns) {constMax=0.5, constMin=-0.25})
  )
  n53 = SmoothStep(Value=MaterialFunctionCall(
    UV=Panner(Coordinate=@n46.Scaled to Fit, Speed=@n52) {speedY=0.25}
    Tiling=ScalarParameter() {defaultValue=4, group=Likes, parameterName=Set1_Tile}
    Cell Offset=Constant()
    Cell Rotation=NamedRerouteUsage()
    Random Seed=@n45
   ) {materialFunction=&MF_UI_ParticleGrid}.Cells UV)
  n54 = ScalarParameter() {defaultValue=2, group=Likes, parameterName=Scale, sortPriority=0}
  n55 = AppendVector(
   A=TextureSample(UVs=MaterialFunctionCall(UVs=@n53, Texture Scale=@n54) {materialFunction=&ScaleUVsByCenter}.UVs) {samplerType=SAMPLERTYPE_LinearColor, texture=/FortUILibrary/Textures/Icons/T_UI_Icon_UIKit_Like_SDF}.RGB
   B=Constant()
  )
  n56 = Multiply(A=MaterialFunctionCall(
    DistanceField=ComponentMask(None=@n55) {r=true}
    Threshold=ScalarParameter() {defaultValue=0.68, group=Likes, parameterName=Fill Aliasing}
    Smoothing=ScalarParameter() {group=2 - Distance Field, parameterName=Smoothing, sliderMax=1}
   ) {materialFunction=&MF_UI_AntiAliasing}.Outer) {constB=0.8}
  n57 = ScalarParameter() {defaultValue=5, group=Likes, parameterName=Set2_Tile}
  n58 = Constant()
  n59 = ScalarParameter() {defaultValue=0.2, group=Likes, parameterName=Set1_Y_PanSpeed}
  n60 = NamedRerouteUsage()
  n61 = Constant()
  n62 = Constant() {r=2.9415}
  n63 = Multiply(
   A=LinearInterpolate(A=Multiply(A=@n59) {constB=-1}, B=@n59, Alpha=ScalarParameter() {defaultValue=1, group=Likes, parameterName=Up})
   B=SmoothStep(Value=MaterialFunctionCall(UV=@n60, Tiling=@n57, Cell Offset=@n61, Cell Rotation=@n61, Random Seed=@n62) {materialFunction=&MF_UI_ParticleGrid}.Random Columns) {constMax=0.5, constMin=-0.25}
  )
  n64 = MaterialFunctionCall(
   UV=Panner(
    Coordinate=MaterialFunctionCall() {materialFunction=&MF_UI_UVScaling}.Scaled to Fit
    Speed=AppendVector(A=@n58, B=@n63)
   ) {speedY=0.25}
   Tiling=@n57
   Cell Offset=Constant()
   Cell Rotation=NamedRerouteUsage()
   Random Seed=@n45
  ) {materialFunction=&MF_UI_ParticleGrid}
  n65 = TextureSample(UVs=MaterialFunctionCall(UVs=SmoothStep(Value=@n64.Cells UV), Texture Scale=ScalarParameter() {defaultValue=2, group=Likes, parameterName=Scale, sortPriority=0}) {materialFunction=&ScaleUVsByCenter}.UVs) {samplerType=SAMPLERTYPE_LinearColor, texture=&T_UI_Icon_UIKit_Like_SDF}
  n66 = Constant()
  n67 = MaterialFunctionCall(
   DistanceField=ComponentMask(None=AppendVector(A=@n65.RGB, B=@n66)) {r=true}
   Threshold=ScalarParameter() {defaultValue=0.68, group=Likes, parameterName=Fill Aliasing}
   Smoothing=ScalarParameter() {group=2 - Distance Field, parameterName=Smoothing, sliderMax=1}
  ) {materialFunction=&MF_UI_AntiAliasing}
  n68 = NamedRerouteUsage()
  n69 = Constant2Vector() {g=0.7, r=0.5}
  n70 = Constant() {r=4}
  NamedRerouteDeclaration(None=LinearInterpolate(
    B=LinearInterpolate(A=@n56, B=@n67.Outer, Alpha=@n67.Outer)
    Alpha=MaterialFunctionCall(UVs=@n68, CenterPosition=@n69, Density=@n70) {materialFunction=&RadialGradientExponential}.RadialGradientExponential
   ) {constB=0}) {name=Likes, nodeColor=<rgba 0 1 0.2969 1>}
  n71 = NamedRerouteUsage()
  n72 = ScalarParameter() {parameterName=ShowLikes}
  n73 = ScalarParameter() {defaultValue=0.07, group=Likes, parameterName=RayStep}
  n74 = ScalarParameter() {defaultValue=0.09, group=Likes, parameterName=RaySize}
  n75 = Add(A=@n73, B=@n74)
  n76 = Add(A=@n73, B=@n75)
  n77 = Constant() {r=0.9}
  MaterialExpressionRemap_0 = MaterialXRemap!(None=NamedRerouteUsage(), InputLow=@n75, InputHigh=@n76, TargetLow, TargetHigh=@n77) {inputHighDefault=0.4, inputLowDefault=0.3}
  n78 = Saturate(None=@MaterialExpressionRemap_0)
  n79 = Constant() {r=0.01}
  n80 = Saturate(None=MaterialXRemap(None=NamedRerouteUsage(), InputLow=Subtract(A=@n76, B=@n79), InputHigh=@n76, TargetHigh=@n77) {inputHighDefault=0.4, inputLowDefault=0.39})
  n81 = NamedRerouteUsage()
  n82 = Constant() {r=0.5}
  n83 = Constant() {r=5}
  n84 = Saturate(None=MaterialXRemap(None=NamedRerouteUsage(), InputLow=Subtract(A=@n75, B=@n73) {constB=0.01}, InputHigh=@n75, TargetHigh=@n77) {inputHighDefault=0.3, inputLowDefault=0.2})
  n85 = Saturate(None=MaterialXRemap(None=NamedRerouteUsage(), InputLow=Subtract(A=@n75, B=@n79) {constB=0.01}, InputHigh=@n75, TargetHigh=@n77) {inputHighDefault=0.3, inputLowDefault=0.29})
  n86 = NamedRerouteUsage()
  n87 = Constant() {r=0.55}
  n88 = Constant() {r=5}
  n89 = Add(
   A=LinearInterpolate(
    B=Subtract(A=@n78, B=@n80)
    Alpha=MaterialFunctionCall(UVs=@n81, Radius=@n82, Density=@n83) {materialFunction=&RadialGradientExponential}.RadialGradientExponential
   )
   B=LinearInterpolate(
    B=Subtract(A=@n84, B=@n85)
    Alpha=MaterialFunctionCall(UVs=@n86, Radius=@n87, Density=@n88) {materialFunction=&RadialGradientExponential}.RadialGradientExponential
   )
  )
  n90 = MaterialXRemap(None=NamedRerouteUsage(), InputHigh=@n74, TargetHigh=@n77) {inputHighDefault=0.11, targetLowDefault=0.7}
  n91 = MaterialXRemap(None=NamedRerouteUsage(), InputLow=Subtract(A=@n74, B=@n79) {constB=0.01}, InputHigh=@n74, TargetHigh=@n77) {inputHighDefault=0.11, inputLowDefault=0.09}
  n92 = LinearInterpolate(
   B=Subtract(A=Saturate(None=@n90), B=Saturate(None=@n91))
   Alpha=MaterialFunctionCall(UVs=NamedRerouteUsage(), Radius=Constant() {r=0.7}, Density=Constant() {r=3}) {materialFunction=&RadialGradientExponential}.RadialGradientExponential
  )
  n93 = Multiply(A=Saturate(None=Add(A=@n89, B=@n92)), B=ScalarParameter() {defaultValue=0.5, group=Likes, parameterName=RaysIntensity}) {constB=0.5}
  n94 = NamedRerouteUsage()
  NamedRerouteDeclaration(None=Multiply(
    A=MaterialFunctionCall(UVs=@n71, Radius=@n72) {materialFunction=&RadialGradientExponential}.RadialGradientExponential
    B=Multiply(A=@n93, B=@n94)
   )) {name=Rays, nodeColor=<rgba 0 1 0.8359 1>}
  n95 = Constant() {r=0.1}
  n96 = Constant() {r=0.4}
  n97 = Constant() {r=0.75}
  n98 = Constant() {r=0.1}
  n99 = Add(
   A=Max(
    A=MaterialFunctionCall(GradientStart=@n95, GradientEnd=@n96, RotationAngle=@n97, GradientOffset=@n98) {materialFunction=&MF_UI_LinearGradient}.Result
    B=NamedRerouteUsage()
   )
   B=NamedRerouteUsage()
  )
  n100 = NamedRerouteUsage()
  n101 = NamedRerouteUsage()
  n102 = LinearInterpolate(A=@n99, B=Add(A=@n99, B=Multiply(A=@n100, B=@n101)), Alpha=ScalarParameter() {parameterName=UseTriangles}) {constAlpha=0}
  n103 = LinearInterpolate(A=NamedRerouteUsage(), B=NamedRerouteUsage(), Alpha=Max(A=NamedRerouteUsage(), B=ScalarParameter() {parameterName=UseOutline}))
  n104 = NamedRerouteUsage()
  n105 = LinearInterpolate(A=@n104, B=NamedRerouteUsage(), Alpha=NamedRerouteUsage())
  n106 = NamedRerouteUsage()
  n107 = MaterialFunctionCall!(
   Base=LinearInterpolate(A=Multiply(A=@n102, B=@n103), B=LinearInterpolate(A=@n105, B=@n104, Alpha=@n106), Alpha=NamedRerouteUsage())
   Blend=NamedRerouteUsage()
  ) {materialFunction=/Engine/Functions/Engine_MaterialFunctions03/Blends/Blend_SoftLight}
  n108 = Saturate(None=LinearInterpolate(B=NamedRerouteUsage(), Alpha=NamedRerouteUsage()) {constB=0})
  n109 = Multiply(A=@n108, B=NamedRerouteUsage())
  n110 = Multiply(A=@n108) {constB=0.7}
  NamedRerouteDeclaration(None=LinearInterpolate(A=@n107.Result, B=LinearInterpolate(A=@n107.Result, B=@n109, Alpha=@n110), Alpha=ScalarParameter() {parameterName=ShowLikes})) {name=MainColor, nodeColor=<rgba 0 0.4609 1 1>}
  n111 = NamedRerouteUsage()
  n112 = AppendVector(A=ComponentMask(None=@n111) {r=true}, B=Subtract(A=ComponentMask(None=@n111) {g=true}) {constB=0.4})
  NamedRerouteDeclaration(None=@n112) {name=RaysUV, nodeColor=<rgba 0.9688 1 0 1>}
  NamedRerouteDeclaration(None=MaterialFunctionCall(
    DistanceField=NamedRerouteUsage()
    Threshold=Constant() {r=0.5}
    Smoothing=ScalarParameter() {group=2 - Distance Field, parameterName=Smoothing}
   ) {materialFunction=&MF_UI_AntiAliasing}.Outer) {name=BannerInnerShape, nodeColor=<rgba 0 0.2969 1 1>}
  n113 = MaterialFunctionCall!(Vector or UVs=@n112) {materialFunction=/FortUILibrary/MaterialFunctions/Basics/MF_UI_CheaperVectorToRadial}
  n114 = OneMinus(None=@n113.Vector Converted to Angle)
  NamedRerouteDeclaration(None=Multiply(A=Min(A=@n113.Vector Converted to Angle, B=@n114)) {constB=2}) {name=RaysShape, nodeColor=<rgba 0 1 0.4375 1>}
  NamedRerouteDeclaration(None=ScalarParameter() {parameterName=AlternativeStyle}) {name=UseAlternativeStyle, nodeColor=<rgba 1 0 0.7266 1>}
  n115 = Constant() {r=1.5}
  NamedRerouteDeclaration(None=Max(
    A=MaterialFunctionCall(UVs=@n17.UVs, Density=@n115) {materialFunction=&RadialGradientExponential}.RadialGradientExponential
    B=NamedRerouteUsage()
   )) {name=AlternativeColorCircleGradient, nodeColor=<rgba 0.9453 0 1 1>}
  n116 = MaterialFunctionCall!(DistanceField=NamedRerouteUsage(), Threshold, Thickness=Constant() {r=0.0508}, Alignment=Constant() {r=10}, Smoothing=Constant() {r=0.8784}) {materialFunction=/FortUILibrary/MaterialFunctions/Shapes/Modifiers/MF_UI_ShapeOutline}
  n117 = NamedRerouteUsage()
  NamedRerouteDeclaration(None=Multiply(A=Multiply(A=@n116.Outline, B=@n117)) {constB=0.8}) {name=InnerOutline_Gradient, nodeColor=<rgba 0 1 0.1094 1>}
  n118 = MaterialFunctionCall!(UVs=TextureCoordinate()) {materialFunction=/FortUILibrary/MaterialFunctions/Basics/MF_UI_9SliceUVs}
  n119 = Constant2Vector() {g=0.75, r=0.75}
  n120 = TextureSample(UVs=Add(
    A=MaterialFunctionCall(UVs=@n118.UVs, Texture Scale=@n119) {materialFunction=&ScaleUVsByCenter}.UVs
    B=Constant2Vector() {g=0.03}
   )) {
   samplerSource=SSM_Clamp_WorldGroupSettings
   samplerType=SAMPLERTYPE_LinearColor
   texture=/Game/UI/Frontend/Lobby/Habanero/Habanero_Textures/T_UI_Habanero_Banner_MSDF
  }
  NamedRerouteDeclaration(None=MaterialFunctionCall(MSDF RGB=@n120.RGB) {materialFunction=&MF_UI_MSDFtoSDF}.SDF) {name=MainShapeSdf, nodeColor=<rgba 0 0.8594 1 1>}
  n121 = Constant() {r=0.35}
  n122 = Constant() {r=0.7}
  n123 = Constant() {r=0.25}
  NamedRerouteDeclaration(None=MaterialFunctionCall(
    A=VectorParameter() {parameterName=AlternativeColor_1, defaultValue=<rgba 0.4287 0.0666 0 1>}.RGB
    B=VectorParameter() {parameterName=AlternativeColor_2, defaultValue=<rgba 1 0.4793 0.1356 1>}.RGB
    C=NamedRerouteUsage()
    Alpha=MaterialFunctionCall(GradientStart=@n121, GradientEnd=@n122, RotationAngle=@n123) {materialFunction=&MF_UI_LinearGradient}.Result
   ) {materialFunction=&3ColorBlend}.Result) {name=AlternativeOutlineColor, nodeColor=<rgba 1 0 0.8906 1>}
  NamedRerouteDeclaration(None=LinearInterpolate(
    A=VectorParameter() {parameterName=Color, defaultValue=<rgba 0.15 0.0467 0.0075 1>}.RGB
    B=VectorParameter() {parameterName=Anim_Color, defaultValue=<rgba 0.15 0.0467 0.0075 1>}.RGB
    Alpha=ScalarParameter() {parameterName=AnimColorTransition}
   )) {name=Color, nodeColor=<rgba 0 1 0.5781 1>}
  n124 = AppendVector(A=ScalarParameter() {group=Texture, parameterName=Anchor X}, B=ScalarParameter() {defaultValue=1, group=Texture, parameterName=Anchor Y})
  n125 = Subtract(A=TextureCoordinate(), B=@n124)
  n126 = MaterialFunctionCall() {materialFunction=&GetUserInterfaceUV}
  n127 = ComponentMask(None=@n126.Pixel Size) {g=true}
  n128 = ComponentMask(None=@n126.Pixel Size) {r=true}
  n129 = Constant() {r=1}
  n130 = Divide(
   A=AppendVector(A=Constant() {r=1}, B=Divide(A=@n127, B=@n128)) {desc=Widget Vertical Aspect Ratio}
   B=AppendVector(A=@n129, B=@n129) {desc=Texture Vertical Aspect Ratio}
  )
  NamedRerouteDeclaration(None=Add(A=Multiply(A=@n125, B=@n130) {desc="Scaling to fix Aspect Ratio, fit to width"}, B=@n124)) {name=UVs Likes, nodeColor=<rgba 1 0.0938 0 1>}
  n131 = TextureCoordinate()
  n132 = Sine(None=Multiply(A=ComponentMask(None=@n131) {g=true}) {constB=2})
  n133 = Clamp(None=@n132)
  n134 = Multiply(A=Clamp(None=@n132) {maxDefault=0, minDefault=-1}) {constB=-1}
  NamedRerouteDeclaration(None=LinearInterpolate(Alpha=Add(A=@n133, B=@n134)) {constA=0.05, constB=-0.05}) {name=Cell Rotation, nodeColor=<rgba 0 0.2266 1 1>}
  n135 = Add(A=NamedRerouteUsage(), B=NamedRerouteUsage())
  n136 = ScalarParameter() {defaultValue=0.8, parameterName=GradientsOpacity}
  NamedRerouteDeclaration(None=LinearInterpolate(A=Multiply(A=@n135, B=@n136), B=NamedRerouteUsage(), Alpha=NamedRerouteUsage())) {name=CenterGradient, nodeColor=<rgba 1 0.2578 0 1>}
  NamedRerouteDeclaration(None=MaterialFunctionCall(DistanceField=NamedRerouteUsage(), Thickness=Constant() {r=0.07}, Alignment=Constant() {r=3.5}) {materialFunction=&MF_UI_ShapeOutline}.Outline) {name=OutlineAlpha, nodeColor=<rgba 0 1 0.8125 1>}
  n137 = MaterialFunctionCall(Texture Scale=Constant2Vector() {g=0.4, r=1}) {materialFunction=&ScaleUVsByCenter}
  n138 = ScalarParameter() {defaultValue=0.7, group=Gradient, parameterName=SideGradientOffset}
  n139 = Constant() {r=0.25}
  n140 = NamedRerouteUsage()
  n141 = Multiply(A=NamedRerouteUsage()) {constB=2}
  NamedRerouteDeclaration(None=Multiply(
    A=MaterialFunctionCall!(UVs=@n137.UVs, Offset=@n138, RotationCenter, RotationAngle=@n139) {materialFunction=/FortUILibrary/MaterialFunctions/Gradients/MF_UI_ReflectedGradient}.Result
    B=Max(A=@n140, B=@n141)
   )) {name=OutlineGlare, nodeColor=<rgba 0 1 0.5547 1>}
  NamedRerouteDeclaration(None=MaterialFunctionCall(DistanceField=NamedRerouteUsage(), Thickness=Constant() {r=0.0508}, Alignment=Constant() {r=10}) {materialFunction=&MF_UI_ShapeOutline}.Outline) {name=InnerOutline, nodeColor=<rgba 0 1 0.1094 1>}
  NamedRerouteDeclaration(None=LinearInterpolate(
    A=VectorParameter() {parameterName=GradientColor_Inner, defaultValue=<rgba 1 0.5711 0 1>}.RGB
    B=VectorParameter() {parameterName=Anim_GradientColor_Inner, defaultValue=<rgba 1 0.5711 0 1>}.RGB
    Alpha=ScalarParameter() {parameterName=AnimColorTransition}
   )) {name=GradientColor_Inner, nodeColor=<rgba 0 1 0.5781 1>}
  NamedRerouteDeclaration(None=MaterialFunctionCall(GradientStart=Constant(), GradientEnd=Constant() {r=0.25}, RotationAngle=Constant() {r=0.75}) {materialFunction=&MF_UI_LinearGradient}.Result) {name=DarkGradient, nodeColor=<rgba 1 0.4453 0 1>}
 ]
}
