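# Translucent UI material for a gradient in the mode info header of the Habanero frontend lobby.
# It takes the vertical component of UV channel 1, inverts it and applies a Power curve, producing a procedural falloff from 1 at V=0 to 0 at V=1 without texture samples.
# A constant of one sits beside the gradient chain; the graph does not show which material outputs either feeds.
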Material /Game/UI/Frontend/Lobby/Habanero/ModeInfo/Textures/M_ModeInfoHeader_extGradient
graph {
 properties {bAutomaticallySetUsageInEditor=false, blendMode=BLEND_Translucent, materialDomain=MD_UI, usedShadingModels=Default Lit}
 nodes [
  Constant() {r=1}
  n1 = ComponentMask(None=TextureCoordinate() {coordinateIndex=1}) {g=true}
  Reroute(None=Power(Base=OneMinus(None=@n1)))
 ]
}
