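# Translucent UI material for detail icons in the Habanero lobby mode info, with a loss protection icon as the default texture.
# The Texture Input parameter supplies an MSDF texture, which MF_UI_MSDFtoSDF converts to a signed distance field.
# MF_UI_AntiAliasing thresholds that field into a crisp icon shape.
# A constant yellow color accompanies the shape, and only the texture is exposed as a parameter.
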Material /Game/UI/Frontend/Lobby/Habanero/ModeInfo/Textures/M_UI_Icon_Details
graph {
 properties {bAutomaticallySetUsageInEditor=false, blendMode=BLEND_Translucent, materialDomain=MD_UI, usedShadingModels=Default Lit}
 nodes [
  Constant3Vector() {constant=<rgba 1 0.8769 0 1>}
  MaterialFunctionCall(
   DistanceField=MaterialFunctionCall!(MSDF RGB=TextureSampleParameter2D() {parameterName=Texture Input, texture=&/T_UI_Icon_LossProtection_MSDF}.RGBA) {materialFunction=/FortUILibrary/MaterialFunctions/Shapes/Modifiers/MF_UI_MSDFtoSDF}.SDF
   Threshold=Constant() {r=0.1}
  ) {materialFunction=&MF_UI_AntiAliasing}
 ]
}
