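# Habanero tier badge widget for the lobby mode info UI, derived from HabaneroDisplayBadge, showing a tier's icon with a text label.
# SetBadgeInternal breaks the tier struct and passes its small foreground and background icons and name to SetupTile.
# SetupTile feeds the nested WBP_Habanero_Icon and styles the bottom-right label, forcing outline alpha to half.
# The event leaves both color inputs unconnected, so they use pin defaults.
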WidgetBlueprint /Game/UI/Frontend/Lobby/Habanero/ModeInfo/WBP_ModeInfo_Icon
graphs [
 EventGraph {nodes [
   n1 = Event SetBadgeInternal!() -> (OutputDelegate, then, TierData:"Fort Habanero Tier Structure (by ref)") {is=title K2Node_Event}
   n2 = Break Fort Habanero Tier!(FortHabaneroTier:"Fort Habanero Tier Structure (by ref)"=@n1.TierData) -> (
    ForegroundIcon:Material Instance Soft Object Reference
    ForegroundIconSmall:Material Instance Soft Object Reference
    BackgroundIcon:Material Instance Soft Object Reference
    BackgroundIconSmall:Material Instance Soft Object Reference
    WhiteIcon:Texture 2D Soft Object Reference
    SimpleIconSmall:Texture 2D Soft Object Reference
    Name:Text
    TierType:Fort Habanero Tier Type Data Object Reference
    AdditionalLeftIcon:Material Instance Soft Object Reference
    AdditionalRightIcon:Material Instance Soft Object Reference
   ) {is=title K2Node_BreakStruct}
   > SetupTile!(
    execute
    self
    IconForeground:Material Instance Soft Object Reference=@n2.ForegroundIconSmall
    IconBackground:Material Instance Soft Object Reference=@n2.BackgroundIconSmall
    Label:Text=@n2.Name
    Label Color:Color Structure
    OutlineColor:Color Structure
   ) -> (then)
  ]}
 SetupTile {nodes [
   n1 = SetupTile!() -> (
    then
    IconForeground:Material Instance Soft Object Reference
    IconBackground:Material Instance Soft Object Reference
    Label:Text
    Label Color:Color Structure
    OutlineColor:Color Structure
   ) {is=title K2Node_FunctionEntry}
   > SetIcon!(
    execute
    self:WBP Habanero Icon Object Reference=Get WBP_Habanero_Icon() -> (WBP_Habanero_Icon:WBP Habanero Icon Object Reference)
    ForegroundInstance:Material Instance Soft Object Reference=@n1
    BackgroundInstance:Material Instance Soft Object Reference=@n1.IconBackground
   ) -> (then)
   n2 = Get Label() -> (Label:Common Text Object Reference)
   > "SetText (Text)"(self=@n2, InText=@n1.Label)
   > SetColorAndOpacity(
    self=@n2
    InColorAndOpacity=Make Slate Color(SpecifiedColor="To LinearColor (Color)"!(InColor:Color Structure=@n1.Label Color) -> (ReturnValue:Linear Color Structure))
   )
   n3 = BreakColor(InColor="To LinearColor (Color)"(InColor=@n1.OutlineColor))
   > Set Outline Color!(
    In:Exec
    CommonTextWidget:Common Text Object Reference=@n2
    Color:Linear Color Structure=MakeColor!(R:float=@n3, G:float=@n3.G, B:float=@n3.B, A:float=0.5) -> (ReturnValue:Linear Color Structure)
   ) -> (Out:Exec) {is=title K2Node_MacroInstance}
  ]}
]
variables []
widgets {
 info {inheritedWidgetCount=0, parentClass=/Script/FortniteGame.HabaneroDisplayBadge, rootWidgetClass=&GridPanel, widgetCount=3}
 widgets [
  Root {widgetClassPath=&GridPanel, widget=Root, properties {bOverride_Cursor=true, columnFill [1], rowFill [1]}}
  WBP_Habanero_Icon {
   widgetClassPath=/Game/UI/Frontend/Lobby/Habanero/WBP_Habanero_Icon
   bIsVariable=true
   parent=Root
   widget=WBP_Habanero_Icon
   properties {useBackground=false, iconMinSize=<115 115>}
   slot {class=GridSlot}
  }
  Label {
   widgetClassPath=&CommonTextBlock
   bIsVariable=true
   parent=Root
   widget=Label
   properties {
    justification=Center
    minDesiredWidth=32
    text=ll
    font {
     bForceMonospaced=false
     bMaterialIsStencil=false
     fontMaterial=null
     fontObject=&BurbankSmall
     letterSpacing=0
     monospacedWidth=1
     size=18
     skewAmount=0
     typefaceFontName=Bold
     outlineSettings {
      bApplyOutlineToDropShadows=false
      bMiteredCorners=false
      bSeparateFillAlpha=false
      outlineMaterial=null
      outlineSize=3
      outlineColor=<rgba 0 0 0 1>
     }
    }
    shadowOffset=<0 0>
   }
   slot {class=GridSlot, properties {horizontalAlignment=HAlign_Right, layer=5, verticalAlignment=VAlign_Bottom, padding=<ltrb 0 0 0 10>}}
  }
 ]
}
