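# Banner widget for one Habanero tier category in the lobby mode info UI, showing a label over a vertical stack of tier icons.
# UpdateList takes a name and tier array, then creates a WBP_ModeInfo_Icon per tier, with numeral labels only when several tiers exist.
# It then tints the background's dynamic material from tier type data and sets a white-icon silhouette.
# An outline is enabled for HabaneroSauce8 when Top500 is set.
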WidgetBlueprint /Game/UI/Frontend/Lobby/Habanero/ModeInfo/WBP_ModeInfo_CategoryDisplay
graphs [
 EventGraph {nodes [
   Event PreConstruct()
   Event Construct()
   Event Tick()
  ]}
 UpdateList {nodes [
   n1 = UpdateList!() -> (then, Name:Text, List:Array of Fort Habanero Tier Structures, Top500:bool=~) {is=title K2Node_FunctionEntry}
   > n2 = Set Ranks(Ranks:Array of Fort Habanero Tier Structures=@n1.List) -> (Output_Get:Array of Fort Habanero Tier Structures)
   > Reverse!(execute, TargetArray:Array of Fort Habanero Tier Structures=@n2) -> (then) {is=title K2Node_CallArrayFunction}
   > "SetText (Text)"(self=Get Label() -> (Label:Common Text Object Reference), InText=@n1)
   > n3 = Reverse for Each Loop!(Exec:Exec, Array:Array of Fort Habanero Tier Structures=Get Ranks() -> (Ranks:Array of Fort Habanero Tier Structures)) -> (LoopBody:Exec, ArrayIndex:int, ArrayElement:Fort Habanero Tier Structure, Completed:Exec) {is=title K2Node_MacroInstance}
   n4 = Get List() -> (List:Array of Fort Habanero Tier Structures)
   n5 = "integer > integer"(A=Length(TargetArray:Array of Fort Habanero Tier Structures=@n4), B=1)
   n6 = Break Fort Habanero Tier!(FortHabaneroTier:"Fort Habanero Tier Structure (by ref)"="Get (a ref)"(Array:Array of Fort Habanero Tier Structures=@n4, Dimension 1=Select(Option 0:int=0, Option 1:int=1, Index=@n5) -> (ReturnValue:int)) -> (Output:"Fort Habanero Tier Structure (by ref)")) -> (WhiteIcon:Texture 2D Soft Object Reference, TierType:Fort Habanero Tier Type Data Object Reference) {is=title K2Node_BreakStruct}
   n7 = Get Background() -> (Background:Fort Lazy Image Object Reference)
   SetBrushFromLazyTexture!(
    execute=@n3.Completed
    self:Common Lazy Image Object Reference=@n7
    LazyTexture:"Texture 2D Soft Object Reference (by ref)"=@n6
    bMatchSize:bool=false
   ) -> (then)
   > n8 = GetDynamicMaterial(self=@n7)
   > n9 = Cast To BP_UI_Habanero_TierTypeDataBase!(execute, Object=@n6.TierType) -> (then, CastFailed, AsBP UI Habanero Tier Type Data Base:BP UI Habanero Tier Type Data Base Object Reference) {is=title K2Node_DynamicCast}
   > SetVectorParameterValue(
    self=@n8
    ParameterName=Color
    Value=Get SocialLeaderboard_BaseColor!(self:BP UI Habanero Tier Type Data Base Object Reference=@n9) -> (SocialLeaderboard_BaseColor:Linear Color Structure)
   )
   > SetVectorParameterValue(
    self=@n8
    ParameterName=GradientColor_Inner
    Value=Get SocialLeaderboard_LightClor!(self:BP UI Habanero Tier Type Data Base Object Reference=@n9) -> (SocialLeaderboard_LightClor:Linear Color Structure)
   )
   > SetVectorParameterValue(
    self=@n8
    ParameterName=GradientColor_Outer
    Value=Get SocialLeaderboard_DarkColor!(self:BP UI Habanero Tier Type Data Base Object Reference=@n9) -> (SocialLeaderboard_DarkColor:Linear Color Structure)
   )
   n10 = Break Fort Habanero Tier!(FortHabaneroTier=@n3.ArrayElement) -> (
    ForegroundIcon:Material Instance Soft Object Reference
    ForegroundIconSmall:Material Instance Soft Object Reference
    BackgroundIcon:Material Instance Soft Object Reference
    BackgroundIconSmall:Material Instance Soft Object Reference
    WhiteIcon
    SimpleIconSmall:Texture 2D Soft Object Reference
    Name:Text
    TierType
    AdditionalLeftIcon:Material Instance Soft Object Reference
    AdditionalRightIcon:Material Instance Soft Object Reference
   )
   n11 = Get Category!(self:Fort Habanero Tier Type Data Object Reference=@n10.TierType) -> (Category:EHabaneroTierType Enum=HabaneroSauce0)
   > Branch(Condition=AND Boolean(A="Equal (Enum)"(A:EHabaneroTierType Enum=@n11, B:EHabaneroTierType Enum=HabaneroSauce8), B=Get Top500() -> (Top500:bool)))
   > SetScalarParameterValue(self=@n8, ParameterName=UseOutline, Value=1)
   n12 = Create Widget!(
    execute:Exec=@n3.LoopBody
    Class:User Widget Class Reference=&/WBP_ModeInfo_Icon.WBP_ModeInfo_Icon_C
    OwningPlayer:Player Controller Object Reference
   ) -> (then:Exec, ReturnValue:WBP Mode Info Icon Object Reference)
   > AddChild(self=Get Vbox_Tiles() -> (Vbox_Tiles:Vertical Box Object Reference), Content=@n12)
   > n13 = Cast To BP_UI_Habanero_TierTypeDataBase(Object=@n10.TierType)
   n14 = Find Item!(TargetArray:Array of Fort Habanero Tier Structures=@n1.List, ItemToFind:"Fort Habanero Tier Structure (by ref)"=@n3.ArrayElement) -> (ReturnValue)
   n15 = Get Tier() -> (Tier:Array of Texts)
   n16 = Length(TargetArray=Get Ranks())
   > SetupTile!(
    execute
    self:WBP Mode Info Icon Object Reference=@n12
    IconForeground:Material Instance Soft Object Reference=@n10.ForegroundIconSmall
    IconBackground:Material Instance Soft Object Reference=@n10.BackgroundIconSmall
    Label:Text=Select(
     Option 0:Text
     Option 1:Text="Get (a copy)"(Array:Array of Texts=@n15, Dimension 1=@n14) -> (Output:Text)
     Index="integer > integer"(A=@n16, B=1)
    ) -> (ReturnValue:Text)
    Label Color:Color Structure=Get Text_Color!(self:BP UI Habanero Tier Type Data Base Object Reference=@n13) -> (Text_Color:Color Structure)
    OutlineColor:Color Structure
   ) -> (then)
  ]}
]
variables []
widgets {
 info {inheritedWidgetCount=0, parentClass=&UserWidget, rootWidgetClass=&GridPanel, widgetCount=6}
 widgets [
  Root {widgetClassPath=&GridPanel, widget=Root, properties {bOverride_Cursor=true, columnFill [1], rowFill [1]}}
  Background {
   widgetClassPath=&FortLazyImage
   bIsVariable=true
   parent=Root
   widget=Background
   properties {
    materialTextureParamName=RankSilhouette
    brush {
     bIsDynamicallyLoaded=false
     drawAs=Image
     imageType=FullColor
     mirroring=NoMirror
     resourceName=None
     resourceObject=/Game/UI/Frontend/Lobby/Habanero/Habanero_Materials/M_Habanero_Banner
     tiling=NoTile
     imageSize=<268 510>
     margin=<ltrb 0 0 0 0>
     outlineSettings {
      bUseBrushTransparency=false
      roundingType=HalfHeightRadius
      width=0
      color {colorUseRule=UseColor_Specified, specifiedColor=<rgba 0 0 0 0>}
      cornerRadii=<0 0 0 0>
     }
     tintColor {colorUseRule=UseColor_Specified, specifiedColor=<rgba 1 1 1 1>}
     uVRegion {bIsValid=false, max=<0 0>, min=<0 0>}
    }
    loadingBackgroundBrush {
     bIsDynamicallyLoaded=false
     drawAs=Image
     imageType=FullColor
     mirroring=NoMirror
     resourceName=None
     resourceObject=null
     tiling=NoTile
     imageSize=<32 32>
     margin=<ltrb 0 0 0 0>
     outlineSettings {
      bUseBrushTransparency=false
      roundingType=HalfHeightRadius
      width=0
      color {colorUseRule=UseColor_Specified, specifiedColor=<rgba 0 0 0 0>}
      cornerRadii=<0 0 0 0>
     }
     tintColor {colorUseRule=UseColor_Specified, specifiedColor=<rgba 1 1 1 1>}
     uVRegion {bIsValid=false, max=<0 0>, min=<0 0>}
    }
   }
   slot {class=GridSlot, properties {horizontalAlignment=HAlign_Center, padding=<ltrb -35 0 -35 0>}}
  }
  VerticalBox_25 {widgetClassPath=&VerticalBox, parent=Root, widget=VerticalBox_25, slot {class=GridSlot, properties {padding=<ltrb 0 0 0 80>}}}
  SizeBox_0 {
   widgetClassPath=&SizeBox
   parent=VerticalBox_25
   widget=SizeBox_0
   properties {bOverride_MaxDesiredWidth=true, maxDesiredWidth=170}
   slot {class=VerticalBoxSlot, properties {horizontalAlignment=HAlign_Center, padding=<ltrb 0 0 0 35>}}
  }
  Label {
   widgetClassPath=&CommonTextBlock
   bIsVariable=true
   parent=SizeBox_0
   widget=Label
   properties {
    justification=Center
    mobileFontSizeMultiplier=1.2
    scrollStyle=&TextScrollStyle-Base_C
    text=Label
    textTransformPolicy=ToUpper
    colorAndOpacity {colorUseRule=UseColor_Specified, specifiedColor=<rgba 1 1 1 0.6>}
    font {
     bForceMonospaced=false
     bMaterialIsStencil=false
     fontMaterial=null
     fontObject=&HeadingNow
     letterSpacing=0
     monospacedWidth=1
     size=20.25
     skewAmount=0
     typefaceFontName=75 Medium
     outlineSettings {
      bApplyOutlineToDropShadows=false
      bMiteredCorners=false
      bSeparateFillAlpha=false
      outlineMaterial=null
      outlineSize=0
      outlineColor=<rgba 0 0 0 1>
     }
    }
    shadowOffset=<0 0>
   }
   slot {class=SizeBoxSlot}
  }
  Vbox_Tiles {
   widgetClassPath=&VerticalBox
   bIsVariable=true
   parent=VerticalBox_25
   widget=Vbox_Tiles
   slot {class=VerticalBoxSlot, properties {horizontalAlignment=HAlign_Center, verticalAlignment=VAlign_Center, size {sizeRule=Fill, value=1}}}
  }
 ]
}
