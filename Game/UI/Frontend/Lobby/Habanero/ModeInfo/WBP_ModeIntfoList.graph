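# Ranked (Habanero) Mode Information panel for the lobby, showing a title, description, artwork, rank category grid and static rule notes.
# PreConstruct fills text and images from the Habanero info data asset, and previews a sample mode at design time.
# OnSetGameMode looks up that mode's tiers in RankData, grouping consecutive same-category tiers into CategoryDisplay widgets.
# An optional top-leaderboard tier runs through the same grouping after the loop.
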WidgetBlueprint /Game/UI/Frontend/Lobby/Habanero/ModeInfo/WBP_ModeIntfoList
graphs [
 EventGraph {nodes [
   n1 = Event OnSetGameMode!() -> (OutputDelegate, then, GameMode:String) {is=title K2Node_Event}
   > Update Grid!(execute, self, GameMode:String=@n1.GameMode) -> (then)
   n2 = Event PreConstruct()
   > UpdatePanel!(execute, self, Object In:Object Reference=/Game/UI/Frontend/Lobby/Habanero/HabaneroData/Habanero_HabaneroInfoData) -> (then)
   > Branch(Condition=@n2.IsDesignTime)
   > Update Grid(GameMode=ranked-blastberry-combined)
   Event On Activated()
  ]}
 Update Grid {nodes [
   n1 = Update Grid!() -> (then, GameMode:String) {is=title K2Node_FunctionEntry}
   > Set Index(Index:int=0) -> (Output_Get:int=0)
   > ClearChildren(self=Get UniformGridPanel_85() -> (UniformGridPanel_85:Uniform Grid Panel Object Reference))
   n2 = Get RankData() -> (RankData:Fort Habanero Display Data Object Reference)
   n3 = Break Fort Habanero Mode List!(FortHabaneroModeList:"Fort Habanero Mode List Structure (by ref)"=Find(
     TargetMap:Map of Strings to Fort Habanero Mode List Structures=Get EntryData!(self:Fort Habanero Display Data Object Reference=@n2) -> (EntryData:Map of Strings to Fort Habanero Mode List Structures)
     Key=@n1
    ) -> (Value:Fort Habanero Mode List Structure)) -> (TierList:Array of Fort Habanero Tier Structures, bShowTopLeaderboardIcon:bool=~, TopTierTopLeaderboardPositions:Fort Habanero Tier Structure) {is=title K2Node_BreakStruct}
   > n4 = For Each Loop(Array:Array of Fort Habanero Tier Structures=@n3) -> (Array Element:Fort Habanero Tier Structure)
   n5 = Branch(execute=@n4.Completed, Condition=@n3.bShowTopLeaderboardIcon)
   > n6 = Sequence()
   > n7 = Set Tier(Tier:Fort Habanero Tier Structure=@n3.TopTierTopLeaderboardPositions) -> (Output_Get:Fort Habanero Tier Structure)
   n8 = Cast To BP_UI_Habanero_TierTypeDataBase!(
    execute=[@n7.then, @n16.then]
    Object=Break Fort Habanero Tier!(FortHabaneroTier:"Fort Habanero Tier Structure (by ref)"=Get Tier() -> (Tier:Fort Habanero Tier Structure)) -> (TierType:Fort Habanero Tier Type Data Object Reference) {is=title K2Node_BreakStruct}
   ) -> (then, CastFailed, AsBP UI Habanero Tier Type Data Base:BP UI Habanero Tier Type Data Base Object Reference) {is=title K2Node_DynamicCast}
   > n9 = Sequence()
   > Set CurrentCategory(CurrentCategory:EHabaneroTierType Enum=Get Category!(self:Fort Habanero Tier Type Data Object Reference=@n8) -> (Category:EHabaneroTierType Enum=HabaneroSauce0)) -> (Output_Get:EHabaneroTierType Enum=HabaneroSauce0)
   n10 = Get LastCategory() -> (LastCategory:EHabaneroTierType Enum=HabaneroSauce0)
   n11 = Get CurrentCategory() -> (CurrentCategory:EHabaneroTierType Enum=HabaneroSauce0)
   n12 = "Equal (Integer)"(A=@n4.Array Index)
   > Branch(Condition=OR Boolean(A="Not Equal (Enum)"!(A:EHabaneroTierType Enum=@n11, B:EHabaneroTierType Enum=@n10) -> (ReturnValue:bool), B=@n12))
   > n13 = Branch(Condition=@n12)
   Clear(
    execute=[@n15.then, @n13.then]
    TargetArray:Array of Fort Habanero Tier Structures=Get CategoryList() -> (CategoryList:Array of Fort Habanero Tier Structures)
   )
   > Set LastCategory(LastCategory:EHabaneroTierType Enum=Get CurrentCategory() -> (CurrentCategory=HabaneroSauce0)) -> (Output_Get:EHabaneroTierType Enum=HabaneroSauce0)
   > Set LastCategoryName(LastCategoryName:Text=Get Name!(self:BP UI Habanero Tier Type Data Base Object Reference=@n8) -> (Name:Text)) -> (Output_Get:Text)
   n14 = Create Widget!(
    execute:Exec=[@n13.else, @n6.then_1, @n5.else]
    Class:User Widget Class Reference=&/WBP_ModeInfo_CategoryDisplay.WBP_ModeInfo_CategoryDisplay_C
    OwningPlayer:Player Controller Object Reference
   ) -> (then:Exec, ReturnValue:WBP Mode Info Category Display Object Reference)
   > AddChildToUniformGrid!(
    execute
    self:Uniform Grid Panel Object Reference=Get UniformGridPanel_85()
    Content:Widget Object Reference=@n14
    InRow:int=0
    InColumn:int=Get Index() -> (Index:int=0)
   ) -> (then, ReturnValue:Uniform Grid Slot Object Reference)
   > Increment Int(Value=Get Index() -> (Index=0))
   > n15 = UpdateList!(
    execute
    self:WBP Mode Info Category Display Object Reference=@n14
    Name:Text=Get LastCategoryName() -> (LastCategoryName:Text)
    List:Array of Fort Habanero Tier Structures=Get CategoryList()
    Top500:bool=@n3.bShowTopLeaderboardIcon
   ) -> (then)
   Add(
    execute=@n9.then_1
    TargetArray:Array of Fort Habanero Tier Structures=Get CategoryList()
    NewItem:"Fort Habanero Tier Structure (by ref)"=Get Tier()
   ) {is=title K2Node_CallArrayFunction}
   n16 = Set Tier(execute=@n4.LoopBody, Tier=@n4)
  ]}
 UpdateEndListPadding {nodes [
   UpdateEndListPadding!() -> (then, IsFirstEntry:bool, IsLastEntry:bool) {is=title K2Node_FunctionEntry}
   > Return Node()
  ]}
 UpdateFocus {nodes [
   UpdateFocus!() -> (then, IsFocused:bool) {is=title K2Node_FunctionEntry}
   > Return Node()
  ]}
 UpdateImage {nodes [
   n1 = UpdateImage!() -> (then, Object:Object Reference) {is=title K2Node_FunctionEntry}
   > n2 = Cast To BP_UI_Habanero_InfoPanel!(execute, Object=@n1) -> (then, CastFailed, AsBP UI Habanero Info Panel:BP UI Habanero Info Panel Object Reference) {is=title K2Node_DynamicCast}
   > SetBrushFromMaterial(
    self=Get Background() -> (Background:Image Object Reference)
    Material=Get ImageBG!(self:BP UI Habanero Info Panel Object Reference=@n2) -> (ImageBG:Material Instance Object Reference)
   )
   > SetBrushFromMaterial(
    self=Get SideImage() -> (SideImage:Image Object Reference)
    Material=Get Image!(self:BP UI Habanero Info Panel Object Reference=@n2) -> (Image:Material Instance Object Reference)
   )
  ]}
 UpdatePanel {nodes [
   n1 = UpdatePanel!() -> (then, Object In:Object Reference) {is=title K2Node_FunctionEntry}
   > UpdateImage!(execute, self, Object:Object Reference=@n1) -> (then) {is=title K2Node_CallFunction}
   > UpdateText!(execute, self, Object:Object Reference=@n1) -> (then)
   > Return Node()
  ]}
 UpdateText {nodes [
   n1 = UpdateText!() -> (then, Object:Object Reference) {is=title K2Node_FunctionEntry}
   > Set Text!(execute, Text) -> (then, Output_Get)
   > n2 = Cast To BP_UI_Habanero_InfoPanel(Object=@n1)
   > "SetText (Text)"(
    self=Get Text_Title() -> (Text_Title:Common Text Object Reference)
    InText=Get Title!(self:BP UI Habanero Info Panel Object Reference=@n2) -> (Title:Text)
   )
   > n3 = For Each Loop(Array:Array of Texts=Get Details List!(self:BP UI Habanero Info Panel Object Reference=@n2) -> (Details List:Array of Texts)) -> (Array Element:Text)
   SetText(execute=@n3.Completed, self=Get Description() -> (Description:Common Rich Text Block Object Reference), InText=Get Text!() -> (Text))
   Set Text(execute=@n3.LoopBody, Text=Format Text!(Format="INVTEXT(\"{Body}{sentence} \")", Body:Text=Get Text(), sentence:Text=@n3) -> (Result))
  ]}
]
variables [
 RankCategories {type=EHabaneroTierType, default=HabaneroSauce0}
 RankData {type=FortHabaneroDisplayData, default=/Game/UI/Habanero/HabaneroDisplayData}
]
widgets {
 info {inheritedWidgetCount=0, parentClass=/Script/FortniteUI.FortHabaneroInfoEntry, rootWidgetClass=&GridPanel, widgetCount=21}
 widgets [
  Root {widgetClassPath=&GridPanel, widget=Root, properties {bOverride_Cursor=true, columnFill [1], rowFill [1]}}
  Background {widgetClassPath=&Image, bIsVariable=true, parent=Root, widget=Background, slot {class=GridSlot}}
  SideImage {
   widgetClassPath=&Image
   bIsVariable=true
   parent=Root
   widget=SideImage
   properties {brush {
     bIsDynamicallyLoaded=false
     drawAs=Image
     imageType=NoImage
     mirroring=NoMirror
     resourceName=None
     resourceObject=null
     tiling=NoTile
     imageSize=<45 32>
     margin=<ltrb 0 0 0 0>
     outlineSettings {
      bUseBrushTransparency=false
      roundingType=HalfHeightRadius
      width=0
      color {colorUseRule=UseColor_Specified, specifiedColor=<rgba 0 0 0 0>}
      cornerRadii=<0 0 0 0>
     }
     tintColor {colorUseRule=UseColor_Specified, specifiedColor=<rgba 1 1 1 1>}
     uVRegion {bIsValid=false, max=<0 0>, min=<0 0>}
    }}
   slot {class=GridSlot, properties {horizontalAlignment=HAlign_Right}}
  }
  SafeZone_0 {
   widgetClassPath=&SafeZone
   bIsVariable=true
   parent=Root
   widget=SafeZone_0
   slot {class=GridSlot, properties {padding=<ltrb 85 47 60 80>}}
  }
  GridPanel_0 {
   widgetClassPath=&GridPanel
   parent=SafeZone_0
   widget=GridPanel_0
   properties {columnFill [1], rowFill [0,0,0,3]}
   slot {class=SafeZoneSlot}
  }
  Text_Title {
   widgetClassPath=&CommonTextBlock
   bIsVariable=true
   parent=GridPanel_0
   widget=Text_Title
   properties {
    scrollStyle=&TextScrollStyle-Base_C
    text="<Title>"
    textTransformPolicy=ToUpper
    wrapTextAt=1152
    font {
     bForceMonospaced=false
     bMaterialIsStencil=false
     fontMaterial=/Game/UI/Frontend/Lobby/Habanero/ModeInfo/Textures/M_ModeInfoHeader_extGradient
     fontObject=&HeadingNow
     letterSpacing=0
     monospacedWidth=1
     size=51.75
     skewAmount=0
     typefaceFontName=86 Bold Italic
     outlineSettings {
      bApplyOutlineToDropShadows=false
      bMiteredCorners=false
      bSeparateFillAlpha=false
      outlineMaterial=null
      outlineSize=0
      outlineColor=<rgba 0 0 0 1>
     }
    }
    shadowOffset=<0 0>
    strikeBrush {
     bIsDynamicallyLoaded=false
     drawAs=Image
     imageType=FullColor
     mirroring=NoMirror
     resourceName=None
     resourceObject=null
     tiling=NoTile
     imageSize=<32 32>
     margin=<ltrb 0 0 0 0>
     outlineSettings {
      bUseBrushTransparency=false
      roundingType=HalfHeightRadius
      width=0
      color {colorUseRule=UseColor_Specified, specifiedColor=<rgba 0 0 0 0>}
      cornerRadii=<0 0 0 0>
     }
     tintColor {colorUseRule=UseColor_Specified, specifiedColor=<rgba 1 1 1 1>}
     uVRegion {bIsValid=false, max=<0 0>, min=<0 0>}
    }
   }
   slot {class=GridSlot}
  }
  CommonTextBlock_263 {
   widgetClassPath=&CommonTextBlock
   parent=GridPanel_0
   widget=CommonTextBlock_263
   properties {
    text=Mode Information
    textTransformPolicy=ToUpper
    colorAndOpacity {colorUseRule=UseColor_Specified, specifiedColor=<rgba 0.6105 1 0.0356 1>}
    font {
     bForceMonospaced=false
     bMaterialIsStencil=false
     fontMaterial=null
     fontObject=&HeadingNow
     letterSpacing=0
     monospacedWidth=1
     size=25.5
     skewAmount=0
     typefaceFontName=86 Bold Italic
     outlineSettings {
      bApplyOutlineToDropShadows=false
      bMiteredCorners=false
      bSeparateFillAlpha=false
      outlineMaterial=null
      outlineSize=0
      outlineColor=<rgba 0 0 0 1>
     }
    }
    shadowOffset=<0 0>
   }
   slot {class=GridSlot, properties {row=1, padding=<ltrb 0 -24 0 0>}}
  }
  SizeBox_0 {
   widgetClassPath=&SizeBox
   parent=GridPanel_0
   widget=SizeBox_0
   properties {
    bOverride_MaxDesiredHeight=true
    bOverride_MaxDesiredWidth=true
    bOverride_MinDesiredHeight=true
    maxDesiredHeight=130
    maxDesiredWidth=1190
    minDesiredHeight=130
   }
   slot {class=GridSlot, properties {horizontalAlignment=HAlign_Left, row=2, verticalAlignment=VAlign_Top, padding=<ltrb 0 30 0 0>}}
  }
  Description {
   widgetClassPath=&CommonRichTextBlock
   bIsVariable=true
   parent=SizeBox_0
   widget=Description
   properties {
    autoWrapText=true
    bOverrideDefaultStyle=true
    lineHeightPercentage=0.875
    minDesiredWidth=800
    mobileTextBlockScale=1.1
    scrollOrientation=Orient_Vertical
    scrollStyle=&TextScrollStyle-Base_C
    text="<b>Eliminate opponents</> and <b>place highly</> to progress through the ranks. "
    textStyleSet=&/ModeInfoModalRichTextStyles
    defaultTextStyleOverride {
     overflowPolicy=Clip
     transformPolicy=None
     colorAndOpacity {colorUseRule=UseColor_Specified, specifiedColor=<rgba 1 1 1 0.95>}
     font {
      bForceMonospaced=false
      bMaterialIsStencil=false
      fontMaterial=null
      fontObject=&HeadingNow
      letterSpacing=0
      monospacedWidth=1
      size=20.25
      skewAmount=0
      typefaceFontName=74 Regular
      outlineSettings {
       bApplyOutlineToDropShadows=false
       bMiteredCorners=false
       bSeparateFillAlpha=false
       outlineMaterial=null
       outlineSize=0
       outlineColor=<rgba 0 0 0 1>
      }
     }
     highlightColor {colorUseRule=UseColor_Specified, specifiedColor=<rgba 0 0 0 1>}
     highlightShape {
      bIsDynamicallyLoaded=false
      drawAs=Image
      imageType=NoImage
      mirroring=NoMirror
      resourceName=None
      resourceObject=null
      tiling=NoTile
      imageSize=<32 32>
      margin=<ltrb 0 0 0 0>
      outlineSettings {
       bUseBrushTransparency=false
       roundingType=HalfHeightRadius
       width=0
       color {colorUseRule=UseColor_Specified, specifiedColor=<rgba 0 0 0 0>}
       cornerRadii=<0 0 0 0>
      }
      tintColor {colorUseRule=UseColor_Specified, specifiedColor=<rgba 1 1 1 1>}
      uVRegion {bIsValid=false, max=<0 0>, min=<0 0>}
     }
     selectedBackgroundColor {colorUseRule=UseColor_Foreground, specifiedColor=<rgba 1 0 1 1>}
     shadowColorAndOpacity=<rgba 0 0 0 1>
     shadowOffset=<0 0>
     strikeBrush {
      bIsDynamicallyLoaded=false
      drawAs=Image
      imageType=NoImage
      mirroring=NoMirror
      resourceName=None
      resourceObject=null
      tiling=NoTile
      imageSize=<32 32>
      margin=<ltrb 0 0 0 0>
      outlineSettings {
       bUseBrushTransparency=false
       roundingType=HalfHeightRadius
       width=0
       color {colorUseRule=UseColor_Specified, specifiedColor=<rgba 0 0 0 0>}
       cornerRadii=<0 0 0 0>
      }
      tintColor {colorUseRule=UseColor_Specified, specifiedColor=<rgba 1 1 1 1>}
      uVRegion {bIsValid=false, max=<0 0>, min=<0 0>}
     }
     underlineBrush {
      bIsDynamicallyLoaded=false
      drawAs=Image
      imageType=NoImage
      mirroring=NoMirror
      resourceName=None
      resourceObject=null
      tiling=NoTile
      imageSize=<32 32>
      margin=<ltrb 0 0 0 0>
      outlineSettings {
       bUseBrushTransparency=false
       roundingType=HalfHeightRadius
       width=0
       color {colorUseRule=UseColor_Specified, specifiedColor=<rgba 0 0 0 0>}
       cornerRadii=<0 0 0 0>
      }
      tintColor {colorUseRule=UseColor_Specified, specifiedColor=<rgba 1 1 1 1>}
      uVRegion {bIsValid=false, max=<0 0>, min=<0 0>}
     }
    }
   }
   slot {class=SizeBoxSlot}
  }
  UniformGridPanel_85 {
   widgetClassPath=/Script/UMG.UniformGridPanel
   bIsVariable=true
   parent=GridPanel_0
   widget=UniformGridPanel_85
   properties {slotPadding=<ltrb 0 0 20 0>}
   slot {class=GridSlot, properties {horizontalAlignment=HAlign_Left, row=3, verticalAlignment=VAlign_Top, padding=<ltrb 0 20 0 0>}}
  }
  SizeBox_1 {
   widgetClassPath=&SizeBox
   parent=GridPanel_0
   widget=SizeBox_1
   properties {bOverride_MaxDesiredHeight=true, bOverride_MinDesiredHeight=true, maxDesiredHeight=210, minDesiredHeight=120}
   slot {class=GridSlot, properties {row=4, verticalAlignment=VAlign_Top, padding=<ltrb 70 0 70 0>}}
  }
  HorizontalBox_116 {widgetClassPath=&HorizontalBox, parent=SizeBox_1, widget=HorizontalBox_116, slot {class=SizeBoxSlot}}
  Image_112 {
   widgetClassPath=&Image
   bIsVariable=true
   parent=HorizontalBox_116
   widget=Image_112
   properties {brush {
     bIsDynamicallyLoaded=false
     drawAs=Image
     imageType=FullColor
     mirroring=NoMirror
     resourceName=None
     resourceObject=/Game/UI/Frontend/Lobby/Habanero/ModeInfo/Textures/M_UI_Icon_Details
     tiling=NoTile
     imageSize=<50 50>
     margin=<ltrb 0 0 0 0>
     outlineSettings {
      bUseBrushTransparency=false
      roundingType=HalfHeightRadius
      width=0
      color {colorUseRule=UseColor_Specified, specifiedColor=<rgba 0 0 0 0>}
      cornerRadii=<0 0 0 0>
     }
     tintColor {colorUseRule=UseColor_Specified, specifiedColor=<rgba 1 1 1 1>}
     uVRegion {bIsValid=false, max=<0 0>, min=<0 0>}
    }}
   slot {class=HorizontalBoxSlot, properties {horizontalAlignment=HAlign_Left, verticalAlignment=VAlign_Top, padding=<ltrb 0 0 40 0>}}
  }
  VerticalBox_Text {
   widgetClassPath=&VerticalBox
   parent=HorizontalBox_116
   widget=VerticalBox_Text
   slot {class=HorizontalBoxSlot, properties {size {sizeRule=Fill, value=1}}}
  }
  CommonText_LossProtection {
   widgetClassPath=&CommonTextBlock
   parent=VerticalBox_Text
   widget=CommonText_LossProtection
   properties {
    scrollStyle=&TextScrollStyle-Base_C
    text=First Match Loss Protection
    font {
     bForceMonospaced=false
     bMaterialIsStencil=false
     fontMaterial=null
     fontObject=&HeadingNow
     letterSpacing=0
     monospacedWidth=1
     size=20.25
     skewAmount=0
     typefaceFontName=75 Medium
     outlineSettings {
      bApplyOutlineToDropShadows=false
      bMiteredCorners=false
      bSeparateFillAlpha=false
      outlineMaterial=null
      outlineSize=0
      outlineColor=<rgba 0 0 0 1>
     }
    }
    shadowOffset=<0 0>
   }
   slot {class=VerticalBoxSlot, properties {padding=<ltrb 0 0 0 10>}}
  }
  CommonText_LossProtection_Body {
   widgetClassPath=&CommonTextBlock
   parent=VerticalBox_Text
   widget=CommonText_LossProtection_Body
   properties {
    autoWrapText=true
    mobileFontSizeMultiplier=1.2
    renderOpacity=0.9
    scrollOrientation=Orient_Vertical
    scrollStyle=&TextScrollStyle-Base_C
    text=Your first match will not reduce your rank progress. This resets daily at 0:00 UTC.
    font {
     bForceMonospaced=false
     bMaterialIsStencil=false
     fontMaterial=null
     fontObject=&HeadingNow
     letterSpacing=0
     monospacedWidth=1
     size=14.25
     skewAmount=0
     typefaceFontName=74 Regular
     outlineSettings {
      bApplyOutlineToDropShadows=false
      bMiteredCorners=false
      bSeparateFillAlpha=false
      outlineMaterial=null
      outlineSize=0
      outlineColor=<rgba 0 0 0 1>
     }
    }
    shadowOffset=<0 0>
   }
   slot {class=VerticalBoxSlot, properties {size {sizeRule=Fill, value=1}}}
  }
  Spacer_358 {widgetClassPath=&Spacer, parent=HorizontalBox_116, widget=Spacer_358, properties {size=<8 1>}, slot {class=HorizontalBoxSlot}}
  Image {
   widgetClassPath=&Image
   bIsVariable=true
   parent=HorizontalBox_116
   widget=Image
   properties {brush {
     bIsDynamicallyLoaded=false
     drawAs=Image
     imageType=FullColor
     mirroring=NoMirror
     resourceName=None
     resourceObject=/Game/UI/Frontend/Lobby/Habanero/ModeInfo/Textures/MI_UI_Icon_Progression
     tiling=NoTile
     imageSize=<50 50>
     margin=<ltrb 0 0 0 0>
     outlineSettings {
      bUseBrushTransparency=false
      roundingType=HalfHeightRadius
      width=0
      color {colorUseRule=UseColor_Specified, specifiedColor=<rgba 0 0 0 0>}
      cornerRadii=<0 0 0 0>
     }
     tintColor {colorUseRule=UseColor_Specified, specifiedColor=<rgba 1 1 1 1>}
     uVRegion {bIsValid=false, max=<0 0>, min=<0 0>}
    }}
   slot {class=HorizontalBoxSlot, properties {horizontalAlignment=HAlign_Left, verticalAlignment=VAlign_Top, padding=<ltrb 0 0 40 0>}}
  }
  VerticalBox_Text_2 {
   widgetClassPath=&VerticalBox
   parent=HorizontalBox_116
   widget=VerticalBox_Text_2
   slot {class=HorizontalBoxSlot, properties {size {sizeRule=Fill, value=1}}}
  }
  CommonText_LossProtection_2 {
   widgetClassPath=&CommonTextBlock
   parent=VerticalBox_Text_2
   widget=CommonText_LossProtection_2
   properties {
    scrollStyle=&TextScrollStyle-Base_C
    text=Progression
    font {
     bForceMonospaced=false
     bMaterialIsStencil=false
     fontMaterial=null
     fontObject=&HeadingNow
     letterSpacing=0
     monospacedWidth=1
     size=20.25
     skewAmount=0
     typefaceFontName=75 Medium
     outlineSettings {
      bApplyOutlineToDropShadows=false
      bMiteredCorners=false
      bSeparateFillAlpha=false
      outlineMaterial=null
      outlineSize=0
      outlineColor=<rgba 0 0 0 1>
     }
    }
    shadowOffset=<0 0>
   }
   slot {class=VerticalBoxSlot, properties {padding=<ltrb 0 0 0 10>}}
  }
  CommonText_LossProtection_Body_2 {
   widgetClassPath=&CommonTextBlock
   parent=VerticalBox_Text_2
   widget=CommonText_LossProtection_Body_2
   properties {
    autoWrapText=true
    mobileFontSizeMultiplier=1.2
    renderOpacity=0.9
    scrollOrientation=Orient_Vertical
    scrollStyle=&TextScrollStyle-Base_C
    text=Dealing damage to eliminated players is shared across all members of a team. Eliminations later in the match are worth more.
    font {
     bForceMonospaced=false
     bMaterialIsStencil=false
     fontMaterial=null
     fontObject=&HeadingNow
     letterSpacing=0
     monospacedWidth=1
     size=14.25
     skewAmount=0
     typefaceFontName=74 Regular
     outlineSettings {
      bApplyOutlineToDropShadows=false
      bMiteredCorners=false
      bSeparateFillAlpha=false
      outlineMaterial=null
      outlineSize=0
      outlineColor=<rgba 0 0 0 1>
     }
    }
    shadowOffset=<0 0>
   }
   slot {class=VerticalBoxSlot, properties {size {sizeRule=Fill, value=1}}}
  }
 ]
}
